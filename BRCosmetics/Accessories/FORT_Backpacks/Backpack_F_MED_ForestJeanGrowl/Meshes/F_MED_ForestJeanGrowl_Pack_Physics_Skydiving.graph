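# Physics asset for the skydiving variant of the ForestJeanGrowl backpack cosmetic, simulating its straps and pack with capsule bodies.
# Five short strap chains (a, b, d, e, f) hang from base bodies with wide swing and almost no twist.
# Stiff front and back strap chains both end at the three-capsule pack through angularly free joints.
# Leg, pelvis, spine and hilt bodies have no constraints.
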PhysicsAsset /BRCosmetics/Accessories/FORT_Backpacks/Backpack_F_MED_ForestJeanGrowl/Meshes/F_MED_ForestJeanGrowl_Pack_Physics_Skydiving
bodies {
 SLegLeft [{shapeType=Capsule, length=50, radius=8, center=<-19 -0.0018 0.7509>, rotation=<rot -87.387 -0.0528 0.0322>}]
 SLegRight [{shapeType=Capsule, length=50, radius=8, center=<19 -0.0001 0.5288>, rotation=<rot -87.387 -0.0528 0.0322>}]
 Spelvis [{shapeType=Capsule, length=10, radius=10, shapeName=Spelvis_capsule, center=<-5.6243 -2.9117 0>}]
 Sspine_01 [{shapeType=Capsule, length=10, radius=10, shapeName=Spelvis_capsule, center=<13.5481 -0.95 0>}]
 Sspine_02 [{shapeType=Capsule, length=0.505, radius=0.505, shapeName=Sspine_02_capsule, rotation=<rot 0 -90 0>}]
 base_strapBack [{shapeType=Capsule, length=0.505, radius=0.505, shapeName=base_strapBack_capsule, rotation=<rot 0 -90 0>}]
 base_strapFront [{shapeType=Capsule, length=0.505, radius=0.505, shapeName=base_strapFront_capsule, rotation=<rot 0 -90 0>}]
 base_strap_a [{shapeType=Capsule, length=0.505, radius=0.505, shapeName=base_strap_a_capsule, rotation=<rot 0 -90 0>}]
 base_strap_b [{shapeType=Capsule, length=0.505, radius=0.505, shapeName=base_strap_b_capsule, rotation=<rot 0 -90 0>}]
 base_strap_d [{shapeType=Capsule, length=0.505, radius=0.505, shapeName=base_strap_d_capsule, rotation=<rot 0 -90 0>}]
 base_strap_e [{shapeType=Capsule, length=0.505, radius=0.505, shapeName=base_strap_e_capsule, rotation=<rot 0 -90 0>}]
 base_strap_f [{shapeType=Capsule, length=0.505, radius=0.505, shapeName=base_strap_f_capsule, rotation=<rot 0 -90 0>}]
 dyn_strapBack_01 [{shapeType=Capsule, length=2, radius=0.5, shapeName=dyn_strapBack_01_capsule, center=<0 -1.0421 0>, rotation=<rot 0 -90 0>}]
 dyn_strapBack_02 [{
   shapeType=Capsule
   length=10
   radius=0.5
   shapeName=dyn_strapBack_02_capsule
   center=<2.5725 -1.9882 -6.1668>
   rotation=<rot 0.0001 -89.8894 -76.097>
  }]
 dyn_strapBack_03 [{
   shapeType=Capsule
   length=10
   radius=0.5
   shapeName=dyn_strapBack_03_capsule
   center=<5.2922 -1.7271 -9.9211>
   rotation=<rot 0.0001 -89.2837 -81.3669>
  }]
 dyn_strapBack_04 [{
   shapeType=Capsule
   length=10
   radius=0.5
   shapeName=dyn_strapBack_04_capsule
   center=<4.5483 -1.6492 -9.3103>
   rotation=<rot -0.0052 90.6199 -64.6075>
  }]
 dyn_strapBack_05 [{shapeType=Capsule, length=2, radius=0.5, shapeName=dyn_strapBack_05_capsule, rotation=<rot 0 -90 0>}]
 dyn_strapFront_01 [{shapeType=Capsule, length=0.505, radius=1, shapeName=dyn_strapFront_01_capsule, rotation=<rot 0 -90 0>}]
 dyn_strapFront_02 [{
   shapeType=Capsule
   length=10
   radius=1
   shapeName=dyn_strapFront_02_capsule
   center=<2.728 1.4627 -6.4369>
   rotation=<rot 0.0001 -90.0775 -73.0379>
  }]
 dyn_strapFront_03 [{
   shapeType=Capsule
   length=10
   radius=1
   shapeName=dyn_strapFront_03_capsule
   center=<4.9017 1.5183 -10.0213>
   rotation=<rot 0.0001 -89.582 -80.8137>
  }]
 dyn_strapFront_04 [{
   shapeType=Capsule
   length=10
   radius=1
   shapeName=dyn_strapFront_04_capsule
   center=<4.7158 1.6391 -9.3396>
   rotation=<rot 0.1629 -90.0004 64.2811>
  }]
 dyn_strapFront_05 [{shapeType=Capsule, length=2, radius=1, shapeName=dyn_strapFront_05_capsule, rotation=<rot 0 -90 0>}]
 dyn_strap_a_00 [{shapeType=Capsule, length=2, radius=0.5, shapeName=dyn_strap_a_00_capsule, center=<2 0 0>, rotation=<rot 90 0 0>}]
 dyn_strap_a_01 [{shapeType=Capsule, length=2, radius=0.5, shapeName=dyn_strap_a_01_capsule, center=<0.4885 0 0>, rotation=<rot 90 0 0>}]
 dyn_strap_a_02 [{shapeType=Capsule, length=2, radius=0.5, shapeName=dyn_strap_a_02_capsule, rotation=<rot 90 0 0>}]
 dyn_strap_b_00 [{shapeType=Capsule, length=2, radius=0.5, shapeName=dyn_strap_b_00_capsule, center=<2 0 0>, rotation=<rot 90 0 0>}]
 dyn_strap_b_01 [{shapeType=Capsule, length=2, radius=0.5, shapeName=dyn_strap_b_01_capsule, center=<0.2043 0 0>, rotation=<rot 90 0 0>}]
 dyn_strap_b_02 [{shapeType=Capsule, length=2, radius=0.5, shapeName=dyn_strap_b_02_capsule, rotation=<rot 90 0 0>}]
 dyn_strap_d_00 [{shapeType=Capsule, length=2, radius=0.5, shapeName=dyn_strap_d_00_capsule, center=<1 0 0>, rotation=<rot 90 0 0>}]
 dyn_strap_d_01 [{shapeType=Capsule, length=2, radius=0.5, shapeName=dyn_strap_d_01_capsule, center=<2 0 0>, rotation=<rot 90 0 0>}]
 dyn_strap_d_02 [{shapeType=Capsule, length=2, radius=0.5, shapeName=dyn_strap_d_02_capsule, center=<2 0 0>, rotation=<rot 90 0 0>}]
 dyn_strap_e_00 [{shapeType=Capsule, length=2, radius=0.5, shapeName=dyn_strap_e_00_capsule, center=<1 0 0>, rotation=<rot 90 0 0>}]
 dyn_strap_e_01 [{shapeType=Capsule, length=2, radius=0.5, shapeName=dyn_strap_e_01_capsule, center=<2 0 0>, rotation=<rot 90 0 0>}]
 dyn_strap_e_02 [{shapeType=Capsule, length=2, radius=0.5, shapeName=dyn_strap_e_02_capsule, center=<2 0 0>, rotation=<rot 90 0 0>}]
 dyn_strap_e_03 [{shapeType=Capsule, length=2, radius=0.5, shapeName=dyn_strap_e_03_capsule, center=<2 0 0>, rotation=<rot 90 0 0>}]
 dyn_strap_f_00 [{shapeType=Capsule, length=2, radius=0.5, shapeName=dyn_strap_f_00_capsule, center=<1 0 0>, rotation=<rot 90 0 0>}]
 dyn_strap_f_01 [{shapeType=Capsule, length=2, radius=0.5, shapeName=dyn_strap_f_01_capsule, center=<2 0 0>, rotation=<rot 90 0 0>}]
 dyn_strap_f_02 [{shapeType=Capsule, length=2, radius=0.5, shapeName=dyn_strap_f_02_capsule, center=<2 0 0>, rotation=<rot 90 0 0>}]
 dyn_strap_f_03 [{shapeType=Capsule, length=2, radius=0.5, shapeName=dyn_strap_f_03_capsule, center=<2 0 0>, rotation=<rot 90 0 0>}]
 pack [
  {shapeType=Capsule, length=20, radius=2, shapeName=pack_capsule, center=<9.6497 -1.2133 -5.0494>, rotation=<rot -3.8732 0 -0.0002>}
  {shapeType=Capsule, length=20, radius=10, shapeName=pack_capsule, center=<13.0977 13.7208 -4.7247>, rotation=<rot 88.194 -179.9821 -176.6221>}
  {
   shapeType=Capsule
   length=70
   radius=2.5
   shapeName=pack_capsule
   center=<-7.9165 -0.9916 -13.4568>
   rotation=<rot 36.3321 176.7309 178.1764>
  }
 ]
 star_hilt [{shapeType=Capsule, length=7, radius=0.5, shapeName=star_hilt_capsule, center=<0.6512 0 0>}]
 sword_hilt [{shapeType=Capsule, length=30, radius=1.5, shapeName=sword_hilt_capsule, center=<3.6123 0 0>, rotation=<rot 90 0 0>}]
}
constraints [
 {bone1Name=dyn_strap_a_00, bone2Name=base_strap_a, swing1LimitDegrees=65, swing2LimitDegrees=65, twistLimitDegrees=1}
 {bone1Name=dyn_strap_a_01, bone2Name=dyn_strap_a_00, swing1LimitDegrees=65, swing2LimitDegrees=65, twistLimitDegrees=1}
 {bone1Name=dyn_strap_b_00, bone2Name=base_strap_b, swing1LimitDegrees=65, swing2LimitDegrees=65, twistLimitDegrees=1}
 {bone1Name=dyn_strap_b_01, bone2Name=dyn_strap_b_00, swing1LimitDegrees=65, swing2LimitDegrees=65, twistLimitDegrees=1}
 {bone1Name=dyn_strap_d_00, bone2Name=base_strap_d, swing1LimitDegrees=145, swing2LimitDegrees=145, twistLimitDegrees=1}
 {bone1Name=dyn_strap_d_01, bone2Name=dyn_strap_d_00, swing1LimitDegrees=45, swing2LimitDegrees=45, twistLimitDegrees=1}
 {bone1Name=dyn_strap_d_02, bone2Name=dyn_strap_d_01, swing1LimitDegrees=45, swing2LimitDegrees=45, twistLimitDegrees=1}
 {bone1Name=dyn_strap_e_00, bone2Name=base_strap_e, swing1LimitDegrees=125, swing2LimitDegrees=125, twistLimitDegrees=1}
 {bone1Name=dyn_strap_e_01, bone2Name=dyn_strap_e_00, swing1LimitDegrees=45, swing2LimitDegrees=45, twistLimitDegrees=1}
 {bone1Name=dyn_strap_e_02, bone2Name=dyn_strap_e_01, swing1LimitDegrees=45, swing2LimitDegrees=45, twistLimitDegrees=1}
 {bone1Name=dyn_strap_f_00, bone2Name=base_strap_f, swing1LimitDegrees=160, swing2LimitDegrees=160, twistLimitDegrees=1}
 {bone1Name=dyn_strap_f_01, bone2Name=dyn_strap_f_00, swing1LimitDegrees=45, swing2LimitDegrees=45, twistLimitDegrees=1}
 {bone1Name=dyn_strap_f_02, bone2Name=dyn_strap_f_01, swing1LimitDegrees=45, swing2LimitDegrees=45, twistLimitDegrees=1}
 {bone1Name=dyn_strap_f_03, bone2Name=dyn_strap_f_02, swing1LimitDegrees=45, swing2LimitDegrees=45, twistLimitDegrees=1}
 {bone1Name=dyn_strap_e_03, bone2Name=dyn_strap_e_02, swing1LimitDegrees=45, swing2LimitDegrees=45, twistLimitDegrees=1}
 {bone1Name=dyn_strapBack_01, bone2Name=base_strapBack, swing1LimitDegrees=1, swing2LimitDegrees=1, twistLimitDegrees=25}
 {bone1Name=dyn_strapBack_02, bone2Name=dyn_strapBack_01, swing1LimitDegrees=3.5, swing2LimitDegrees=3.5, twistLimitDegrees=5}
 {bone1Name=dyn_strapBack_03, bone2Name=dyn_strapBack_02, swing1LimitDegrees=3.5, swing2LimitDegrees=3.5, twistLimitDegrees=5}
 {bone1Name=dyn_strapBack_04, bone2Name=dyn_strapBack_03, swing1LimitDegrees=3.5, swing2LimitDegrees=3.5, twistLimitDegrees=5}
 {bone1Name=dyn_strapBack_05, bone2Name=dyn_strapBack_04, swing1LimitDegrees=3.5, swing2LimitDegrees=3.5, twistLimitDegrees=5}
 {bone1Name=dyn_strapFront_01, bone2Name=base_strapFront, swing1LimitDegrees=1, swing2LimitDegrees=1, twistLimitDegrees=25}
 {bone1Name=dyn_strapFront_02, bone2Name=dyn_strapFront_01, swing1LimitDegrees=3.5, swing2LimitDegrees=3.5, twistLimitDegrees=5}
 {bone1Name=dyn_strapFront_03, bone2Name=dyn_strapFront_02, swing1LimitDegrees=3.5, swing2LimitDegrees=3.5, twistLimitDegrees=5}
 {bone1Name=dyn_strapFront_04, bone2Name=dyn_strapFront_03, swing1LimitDegrees=3.5, swing2LimitDegrees=3.5, twistLimitDegrees=5}
 {bone1Name=dyn_strapFront_05, bone2Name=dyn_strapFront_04, swing1LimitDegrees=3.5, swing2LimitDegrees=3.5, twistLimitDegrees=5}
 {bone1Name=dyn_strap_a_02, bone2Name=dyn_strap_a_01, swing1LimitDegrees=65, swing2LimitDegrees=65, twistLimitDegrees=1}
 {bone1Name=dyn_strap_b_02, bone2Name=dyn_strap_b_01, swing1LimitDegrees=65, swing2LimitDegrees=65, twistLimitDegrees=1}
 {
  bone1Name=dyn_strapFront_05
  bone2Name=pack
  swing1LimitDegrees=45
  swing1Motion=Free
  swing2LimitDegrees=45
  swing2Motion=Free
  twistLimitDegrees=45
  twistMotion=Free
 }
 {
  bone1Name=dyn_strapBack_05
  bone2Name=pack
  swing1LimitDegrees=45
  swing1Motion=Free
  swing2LimitDegrees=45
  swing2Motion=Free
  twistLimitDegrees=45
  twistMotion=Free
 }
]
