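# Animation Blueprint for the ForestJeanGrowl back bling, simulating its dynamic bones and posing its straps.
# Each update copies the parent mesh's pelvis, spine and thigh socket transforms onto matching S-prefixed bones in world space.
# That cached pose feeds two RigidBody simulations, standard and skydiving, blended by dive or float state before additive strap offsets.
# A material parameter hides an element while a matching tagged pickaxe is equipped.
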AnimBlueprint /BRCosmetics/Accessories/FORT_Backpacks/Backpack_F_MED_ForestJeanGrowl/Meshes/F_MED_ForestJeanGrowl_Pack_AnimBP
target_skeleton=&/F_MED_ForestJeanGrowl_Pack_Skeleton
graphs [
 AnimGraph {nodes [
   n1 = RigidBody!(ComponentPose=Local To Component(LocalPose=Use cached pose 'body_transforms'()), Alpha, OverrideWorldGravity:Vector="0,0,980", ExternalForce) -> (Pose) {
    bClampLinearTranslationLimitToRefPose=true
    bOverrideWorldGravity=true
    bUseExternalClothCollision=true
    overridePhysicsAsset=&/F_MED_ForestJeanGrowl_Pack_Physics
    simulationSpace=BaseBoneSpace
    baseBoneRef {boneName=root}
    componentAppliedLinearAccClamp=<0 0 0>
    simSpaceSettings {
     dampingAlpha=0
     maxAngularAcceleration=15
     maxAngularVelocity=3
     maxLinearAcceleration=1000
     maxLinearVelocity=600
     worldAlpha=1
     externalLinearDragV=<0.25 0.25 0.25>
    }
   }
   n2 = Component To Local(ComponentPose=@n1)
   n3 = RigidBody(ComponentPose=Local To Component(LocalPose=Use cached pose 'body_transforms'()), OverrideWorldGravity="0,0,980") {
    bClampLinearTranslationLimitToRefPose=true
    bOverrideWorldGravity=true
    bUseExternalClothCollision=true
    overridePhysicsAsset=&/F_MED_ForestJeanGrowl_Pack_Physics_Skydiving
    simulationSpace=BaseBoneSpace
    baseBoneRef {boneName=root}
    componentAppliedLinearAccClamp=<0 0 0>
    simSpaceSettings {
     dampingAlpha=0
     maxAngularAcceleration=15
     maxAngularVelocity=3
     maxLinearAcceleration=1200
     maxLinearVelocity=600
     worldAlpha=1
     externalLinearDragV=<0.25 0.25 0.25>
    }
   }
   n4 = Component To Local(ComponentPose=@n3)
   n5 = Get Is Sky Diving_State() -> (Is Sky Diving_State:bool)
   n6 = "Transform (Modify) Bone - Bone: dyn_strapFront_01"!(
    ComponentPose=Local To Component(LocalPose=Blend Poses by bool(BlendPose_0=@n4, BlendPose_1=@n2, BlendTime_0=2, BlendTime_1=0.5, bActiveValue=@n5))
    bAlphaBoolEnabled:bool=True
    Translation="-0.052,0.752,-0.402"
    Rotation="0,4.075,-100"
    Scale="1,1,1"
   ) -> (Pose) {
    alphaInputType=Bool
    rotationMode=BMM_Additive
    rotationSpace=BCS_BoneSpace
    translationMode=BMM_Additive
    translationSpace=BCS_BoneSpace
    alphaBoolBlend {blendInTime=0.5, blendOutTime=0.5}
    boneToModify {boneName=dyn_strapFront_01}
    rotation {roll=-100, yaw=4.075}
    translation=<-0.052 0.7516 -0.4018>
   }
   n7 = "Transform (Modify) Bone - Bone: dyn_strapBack_01"!(ComponentPose=@n6, bAlphaBoolEnabled:bool=True, Translation="-0.052,0.752,-0.402", Rotation="0,4.075,-100", Scale="1,1,1") -> (Pose) {
    alphaInputType=Bool
    rotationMode=BMM_Additive
    rotationSpace=BCS_BoneSpace
    translationMode=BMM_Additive
    translationSpace=BCS_BoneSpace
    alphaBoolBlend {blendInTime=0.5, blendOutTime=0.5}
    boneToModify {boneName=dyn_strapBack_01}
    rotation {roll=-100, yaw=4.075}
    translation=<-0.052 0.752 -0.402>
   }
   n8 = "Transform (Modify) Bone - Bone: dyn_strapFront_01"(ComponentPose=@n7, bAlphaBoolEnabled=True, Translation="-0.052,0.752,-0.402", Rotation="0,4.075,-125", Scale="1,1,1") {
    alphaInputType=Bool
    rotationMode=BMM_Additive
    rotationSpace=BCS_BoneSpace
    translationMode=BMM_Additive
    translationSpace=BCS_BoneSpace
    alphaBoolBlend {blendInTime=0.5, blendOutTime=0.5}
    boneToModify {boneName=dyn_strapFront_01}
    rotation {roll=-125, yaw=4.075}
    translation=<-0.052 0.7516 -0.4018>
   }
   n9 = "Transform (Modify) Bone - Bone: dyn_strapBack_01"(ComponentPose=@n8, bAlphaBoolEnabled=True, Translation="-0.052,0.752,-0.402", Rotation="0,4.075,-105", Scale="1,1,1") {
    alphaInputType=Bool
    rotationMode=BMM_Additive
    rotationSpace=BCS_BoneSpace
    translationMode=BMM_Additive
    translationSpace=BCS_BoneSpace
    alphaBoolBlend {blendInTime=0.5, blendOutTime=0.5}
    boneToModify {boneName=dyn_strapBack_01}
    rotation {roll=-105, yaw=4.075}
    translation=<-0.052 0.752 -0.402>
   }
   Output Pose(Result=Component To Local(ComponentPose=@n9)) {layerGroup=None}
   n10 = "Transform (Modify) Bone - Bone: root"(ComponentPose=Local To Component(LocalPose=Input Pose()), Translation="0,-1.5,0", Rotation="0,5,5") {
    rotationMode=BMM_Additive
    rotationSpace=BCS_BoneSpace
    translationMode=BMM_Additive
    translationSpace=BCS_BoneSpace
    boneToModify {boneName=root}
    rotation {roll=5, yaw=5}
    translation {y=-1.5}
   }
   n11 = "Transform (Modify) Bone - Bone: base_strapFront"(ComponentPose=@n10, Alpha=1, Translation="0,0,0", Rotation="0,0,0", Scale="1,1,1") {
    rotationMode=BMM_Additive
    rotationSpace=BCS_BoneSpace
    translationMode=BMM_Additive
    translationSpace=BCS_BoneSpace
    boneToModify {boneName=base_strapFront}
   }
   Save cached pose 'CRCache_Input'(Pose=Component To Local(ComponentPose=@n11)) {cacheName=CRCache_Input}
   n12 = "Transform (Modify) Bone - Bone: Spelvis"!(
    ComponentPose=Local To Component(LocalPose=Use cached pose 'CRCache_Input'())
    bAlphaBoolEnabled:bool=True
    Translation=Get PelvisLoc() -> (PelvisLoc:Vector)
    Rotation=Get PelvisRot() -> (PelvisRot:Rotator)
    Scale=Get PelvisScale() -> (PelvisScale:Vector)
   ) -> (Pose) {
    alphaInputType=Bool
    rotationMode=BMM_Replace
    rotationSpace=BCS_WorldSpace
    scaleMode=BMM_Replace
    scaleSpace=BCS_WorldSpace
    translationMode=BMM_Replace
    translationSpace=BCS_WorldSpace
    alphaBoolBlend {blendInTime=0.2, blendOption=Linear, blendOutTime=0.2}
    boneToModify {boneName=Spelvis}
    rotation {roll=25}
   }
   n13 = "Transform (Modify) Bone - Bone: Sspine_01"!(
    ComponentPose=@n12
    bAlphaBoolEnabled:bool=True
    Translation=Get spine01_location() -> (spine01_location:Vector)
    Rotation=Get spine01_rotation() -> (spine01_rotation:Rotator)
    Scale="1,1,1"
   ) -> (Pose) {
    alphaInputType=Bool
    rotationMode=BMM_Replace
    rotationSpace=BCS_WorldSpace
    scaleMode=BMM_Replace
    scaleSpace=BCS_WorldSpace
    translationMode=BMM_Replace
    translationSpace=BCS_WorldSpace
    alphaBoolBlend {blendInTime=0.2, blendOption=Linear, blendOutTime=0.2}
    boneToModify {boneName=Sspine_01}
   }
   n14 = "Transform (Modify) Bone - Bone: Sspine_02"!(
    ComponentPose=@n13
    bAlphaBoolEnabled:bool=True
    Translation=Get spine_02_location() -> (spine_02_location:Vector)
    Rotation=Get spine_02_rotation() -> (spine_02_rotation:Rotator)
    Scale="1,1,1"
   ) -> (Pose) {
    alphaInputType=Bool
    rotationMode=BMM_Replace
    rotationSpace=BCS_WorldSpace
    scaleMode=BMM_Replace
    scaleSpace=BCS_WorldSpace
    translationMode=BMM_Replace
    translationSpace=BCS_WorldSpace
    alphaBoolBlend {blendInTime=0.2, blendOption=Linear, blendOutTime=0.2}
    boneToModify {boneName=Sspine_02}
   }
   n15 = "Transform (Modify) Bone - Bone: SLegLeft"!(
    ComponentPose=@n14
    bAlphaBoolEnabled:bool=True
    Translation=Get thigh_left_location() -> (thigh_left_location:Vector)
    Rotation=Get thigh_left_rotation() -> (thigh_left_rotation:Rotator)
    Scale="1,1,1"
   ) -> (Pose) {
    alphaInputType=Bool
    rotationMode=BMM_Replace
    rotationSpace=BCS_WorldSpace
    scaleMode=BMM_Replace
    scaleSpace=BCS_WorldSpace
    translationMode=BMM_Replace
    translationSpace=BCS_WorldSpace
    alphaBoolBlend {blendInTime=0.2, blendOption=Linear, blendOutTime=0.2}
    boneToModify {boneName=SLegLeft}
   }
   n16 = "Transform (Modify) Bone - Bone: SLegRight"!(
    ComponentPose=@n15
    bAlphaBoolEnabled:bool=True
    Translation=Get thigh_right_location() -> (thigh_right_location:Vector)
    Rotation=Get thigh_right_rotation() -> (thigh_right_rotation:Rotator)
    Scale="1,1,1"
   ) -> (Pose) {
    alphaInputType=Bool
    rotationMode=BMM_Replace
    rotationSpace=BCS_WorldSpace
    scaleMode=BMM_Replace
    scaleSpace=BCS_WorldSpace
    translationMode=BMM_Replace
    translationSpace=BCS_WorldSpace
    alphaBoolBlend {blendInTime=0.2, blendOption=Linear, blendOutTime=0.2}
    boneToModify {boneName=SLegRight}
   }
   Save cached pose 'body_transforms'(Pose=Component To Local(ComponentPose=@n16)) {cacheName=body_transforms}
   Get bIsSkyDiving()
  ]}
 EventGraph {nodes [
   Event BlueprintUpdateAnimation()
   > Set Is Sky Diving_State(Is Sky Diving_State:bool=OR Boolean(A=Get bIsSkydiveDiveMode(), B=Get bIsSkydiveFloating())) -> (Output_Get:bool)
   > n1 = Sequence()
   n2 = GetPartSkeletalMeshComponent()
   > Is Valid(InputObject=@n2)
   n3 = GetAttachParent!(self:Scene Component Object Reference=@n2) -> (ReturnValue:Scene Component Object Reference)
   > Is Valid(InputObject=@n3)
   > n4 = Sequence!(execute) -> (then_0, then_1, then_2:Exec, then_3:Exec, then_4:Exec)
   n5 = GetSocketTransform!(self=@n3, InSocketName=pelvis, TransformSpace=RTS_World) -> (ReturnValue_Location:Vector, ReturnValue_Rotation:Rotator, ReturnValue_Scale:Vector)
   > Set PelvisLoc(PelvisLoc:Vector=@n5) -> (Output_Get:Vector)
   > Set PelvisRot(PelvisRot:Rotator=@n5.ReturnValue_Rotation) -> (Output_Get:Rotator)
   > Set PelvisScale(PelvisScale:Vector=@n5.ReturnValue_Scale) -> (Output_Get:Vector)
   n6 = GetSocketTransform(self=@n3, InSocketName=spine_01, TransformSpace=RTS_World)
   Set spine01_location(execute=@n4.then_1, spine01_location:Vector=@n6) -> (Output_Get:Vector)
   > Set spine01_rotation(spine01_rotation:Rotator=@n6.ReturnValue_Rotation) -> (Output_Get:Rotator)
   n7 = GetSocketTransform(self=@n3, InSocketName=spine_02, TransformSpace=RTS_World)
   Set spine_02_location(execute=@n4.then_2, spine_02_location:Vector=@n7) -> (Output_Get:Vector)
   > Set spine_02_rotation(spine_02_rotation:Rotator=@n7.ReturnValue_Rotation) -> (Output_Get:Rotator)
   n8 = GetSocketTransform(self=@n3, InSocketName=thigh_l, TransformSpace=RTS_World)
   Set thigh_left_location(execute=@n4.then_3, thigh_left_location:Vector=@n8) -> (Output_Get:Vector)
   > Set thigh_left_rotation(thigh_left_rotation:Rotator=@n8.ReturnValue_Rotation) -> (Output_Get:Rotator)
   n9 = GetSocketTransform(self=@n3, InSocketName=thigh_r, TransformSpace=RTS_World)
   Set thigh_right_location(execute=@n4.then_4, thigh_right_location:Vector=@n9) -> (Output_Get:Vector)
   > Set thigh_right_rotation(thigh_right_rotation:Rotator=@n9.ReturnValue_Rotation) -> (Output_Get:Rotator)
   n10 = Branch(execute=@n1.then_1, Condition=Get bIsSkyDiving())
   > n11 = Branch(Condition=Get bIsParachuteOpened())
   Set GravityOverride(execute=[@n11.then, @n10.else], GravityOverride="0,0,-980")
   Set GravityOverride(execute=@n11.else, GravityOverride="0,0,980")
   Event OnCharacterCustomizationLoaded()
   > n12 = Sequence!(execute) -> (then_0, then_1, then_2, then_3)
   Set isWearingCape(execute=@n12.then_1, isWearingCape=CheckForBackbling().bIsBackblingACape)
   n13 = Branch(
    execute=[@n15.then, @n12.then_2]
    Condition=PickaxeItemDefinitionHasAnyMetaTag(Tags=Make Array("[0]"=(TagName="Cosmetics.MetaTags.ForestJeanGrowl")), bRequiresEquipped=true)
   )
   > n14 = Get Dynamic Material Instance(SkeletalMeshComponent=GetPartSkeletalMeshComponent())
   > SetScalarParameterValue(self=@n14, ParameterName=Hide Element 01, Value=1)
   SetScalarParameterValue(execute=@n13.else, self=@n14, ParameterName=Hide Element 01, Value=0)
   n15 = Event OnWeaponChanged()
   TryGetPawnOwner()
  ]}
]
variables [
 Body_Mesh {type=SceneComponent, default=null}
 body_root_location {type=Vector}
 body_root_rotation {type=Rotator}
 GravityOverride {type=Vector}
 Is Sky Diving_State {type=boolean}
 isWearingCape {type=boolean}
 pelvis_location {type=Vector}
 pelvis_rotation {type=Rotator}
 PelvisLoc {type=Vector}
 PelvisRot {type=Rotator}
 PelvisScale {type=Vector}
 root_location {type=Vector}
 root_rotation {type=Rotator}
 spine01_location {type=Vector}
 spine01_rotation {type=Rotator}
 spine_02_location {type=Vector}
 spine_02_rotation {type=Rotator}
 thigh_left_location {type=Vector}
 thigh_left_rotation {type=Rotator}
 thigh_right_location {type=Vector}
 thigh_right_rotation {type=Rotator}
]
